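# DontUSE is a placeholder-like Default Lit material in the Asteria forest floor foliage folder, named as one not to be used.
# Its graph holds a single hot-pink color constant, with no connection to a material output shown.
# It is flagged for use with instanced static meshes and Nanite.
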Material /Game/Environments/Asteria/Foliage/ForestFloor/DontUSE
graph {
 properties {bUsedWithInstancedStaticMeshes=true, bUsedWithNanite=true, pixelDepthOffsetMode=PDOM_Legacy, usedShadingModels=Default Lit}
 nodes [Constant3Vector() {constant=<rgba 1 0 0.8916 1>}]
}
