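# Two-sided subsurface material for instanced Nanite ground moss in the Asteria forest floor foliage.
# Two world-textured moss tints get root darkening, hue shift and Fresnel fuzz, with end-to-tip blends for specular, roughness and opacity.
# A NaniteReplace moss gradient uses UVs on Nanite and a baked texture on fallback.
# A switch selects camera-proximity push-down, cancelled while sliding, or destructible-damage offset, then feeds MF_Curie_Landscape in grass mode.
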Material /Game/Environments/Asteria/Foliage/ForestFloor/M_GroundMoss
graph {
 properties {
  bUseMaterialAttributes=true
  bUsedWithInstancedStaticMeshes=true
  bUsedWithNanite=true
  pixelDepthOffsetMode=PDOM_Legacy
  shadingModel=MSM_Subsurface
  twoSided=true
  usedShadingModels=Subsurface
 }
 nodes [
  n1 = MaterialFunctionCall(Color=Constant3Vector()) {materialFunction=&SplitComponents}
  n2 = CollectionParameter() {collection=&FortniteMaterialParameters, parameterName=GrassPushDownRadius}
  n3 = SphereMask(
   A=CameraPositionWS()
   B=TransformPosition(None=Constant3Vector()) {transformSourceType=TRANSFORMPOSSOURCE_Instance, transformType=TRANSFORMPOSSOURCE_World}
   Radius=@n2
   Hardness=Constant()
  )
  n4 = Constant() {r=-200}
  n5 = CollectionParameter() {collection=&FortniteMaterialParameters, parameterName=IsPlayerSliding}
  n6 = Add(A=@n1.B, B=LinearInterpolate(A=Multiply(A=@n3, B=@n4), B=Constant(), Alpha=@n5))
  Add(A=@n6, B=Constant() {r=-2.5}) {desc=move mesh down a bit}
  n7 = ComponentMask(None=WorldPosition().XYZ) {g=true, r=true}
  n8 = MaterialFunctionCall!(
   Texture=TextureObject() {texture=/Game/Effects/FogVolume/T_TileNoise}
   LandscapeCoordinates=Multiply(A=Divide(A=@n7) {constB=10000}, B=NamedRerouteUsage())
   LowEndTiling
   HexGridGradientSharpness=NamedRerouteUsage()
   HexGridSize=NamedRerouteUsage()
   HexGridUVInput
   MinScale=NamedRerouteUsage()
   MaxScale=NamedRerouteUsage()
   MinOffset=NamedRerouteUsage()
   MaxOffset=NamedRerouteUsage()
   "MaxRotation(Degrees)"=NamedRerouteUsage()
   "MinRotation(Degrees)"=NamedRerouteUsage()
   forceLowQuality=NamedRerouteUsage()
  ) {materialFunction=/Game/Athena/Asteria/Environments/Landscape/Materials/MaterialFunctions/UtilityFunctions/Landscape_TextureSample}
  ComponentMask(None=@n8.Result) {a=true}
  ComponentMask(None=TextureCoordinate()) {g=true}
  Constant()
  Constant() {r=1}
  Constant() {r=3}
  Constant() {r=200}
  n9 = Multiply(A=ComponentMask(None=TextureCoordinate()) {g=true}, B=ScalarParameter() {defaultValue=2, parameterName=Frequency})
  Cosine(None=@n9)
  LandscapeLayerCoords()
  n10 = ScalarParameter() {defaultValue=633.3079, parameterName=SizeWorldNormal}
  n11 = MaterialFunctionCall(
   Normal=MaterialFunctionCall(
    TextureObject=TextureObject() {samplerType=SAMPLERTYPE_Normal, texture=/Game/Gadgets/Assets/Telephone_Booth/Textures/T_BlindsNormalNoise_N}
    TextureSize=@n10
   ) {materialFunction=&WorldAlignedNormal}.XYZ Texture
   Flatness=ScalarParameter() {defaultValue=-0.5, parameterName=FlattenWorldNormal}
  ) {materialFunction=&FlattenNormal}
  n12 = Fresnel(ExponentIn=ScalarParameter() {defaultValue=1.3529, parameterName=FresnelExponent}, Normal=Transform(None=@n11.Result)) {baseReflectFraction=0}
  n13 = ScalarParameter() {defaultValue=-0.05, parameterName=MinShift}
  n14 = ScalarParameter() {parameterName=MaxShift}
  n15 = NamedRerouteUsage()
  n16 = LinearInterpolate(
   A=VectorParameter() {parameterName=Moss_BaseColor, defaultValue=<rgba 0.0531 0.1302 0.0142 1>}.RGB
   B=VectorParameter() {parameterName=Moss_BaseColor2, defaultValue=<rgba 0.2552 0.1526 0.0037 1>}.RGB
   Alpha=MaterialFunctionCall(TextureObject=TextureObject() {texture=/Game/Environments/Asteria/Sets/ZenGarden/Textures/T_Asteria_Zen_MossyGround_01_H}, TextureSize=@n10) {materialFunction=&WorldAlignedTexture}.XYZ Texture
  )
  n17 = Add(A=Sine(None=@n9))
  n18 = LinearInterpolate(A=Multiply(A=ScalarParameter() {defaultValue=0.375, parameterName=DarkenRoots}, B=@n16), B=@n16, Alpha=Saturate(None=Divide(A=@n17)))
  n19 = Multiply(
   A=LinearInterpolate(
    A=ScalarParameter() {defaultValue=0.5, parameterName=Fuzz_Darken}
    B=ScalarParameter() {defaultValue=0.76, parameterName=Fuzz_Brighten}
    Alpha=@n12
   )
   B=MaterialFunctionCall(
    Hue Shift Percentage=LinearInterpolate(A=@n13, B=@n14, Alpha=@n15)
    Texture=StaticSwitchParameter(True=@n18, False=@n16) {defaultValue=true, parameterName=UseUVGradient}
   ) {materialFunction=&HueShift}.Result
  )
  LinearInterpolate(
   A=Multiply(A=ScalarParameter() {defaultValue=0.5, parameterName=EndBrightness}, B=@n19)
   B=Multiply(A=ScalarParameter() {defaultValue=1.5, parameterName=TipBrightness}, B=@n19)
   Alpha=NamedRerouteUsage()
  )
  n20 = TextureSampleParameter2D() {parameterName=FallbackNormal, texture=&/SM_Thick_B_Normal}
  n21 = TextureSample() {samplerType=SAMPLERTYPE_Normal, texture=&/T_GrassySpec_N}
  MaterialFunctionCall(BaseNormal=@n20.RGB, AdditionalNormal=@n21.RGB) {materialFunction=&BlendAngleCorrectedNormals}
  MaterialFunctionCall() {materialFunction=&BlendAngleCorrectedNormals}
  n22 = TwoSidedSign()
  n23 = ScalarParameter() {defaultValue=0.1, parameterName=End_Specular}
  n24 = ScalarParameter() {defaultValue=0.2, parameterName=Tip_Specular}
  n25 = NamedRerouteUsage()
  n26 = SetMaterialAttributes!(
   MaterialAttributes
   Base Color=@n19
   Specular=LinearInterpolate(A=LinearInterpolate(A=@n23, B=@n24, Alpha=@n25), B=ScalarParameter() {defaultValue=0.5, parameterName=Fresnel_Specular}, Alpha=@n12)
   Metallic=Constant()
   Roughness=LinearInterpolate(
    A=ScalarParameter() {defaultValue=0.9012, parameterName=End_Roughness}
    B=ScalarParameter() {defaultValue=0.69, parameterName=Tip_Roughness}
    Alpha=NamedRerouteUsage()
   )
   Opacity=LinearInterpolate(
    A=ScalarParameter() {defaultValue=1, parameterName=End_Opacity}
    B=ScalarParameter() {defaultValue=0.5, parameterName=Tip_Opacity}
    Alpha=NamedRerouteUsage()
   )
   Normal=MaterialFunctionCall(BaseNormal=Multiply(A=@n20.RGB, B=@n22), AdditionalNormal=@n11.Result) {materialFunction=&BlendAngleCorrectedNormals}.Result
   Subsurface Color=Multiply(A=@n19, B=ScalarParameter() {defaultValue=0.25, parameterName=SSS})
  ) {attributeSetTypes [
    #1
    #2
    #3
    #4
    #5
    #6
    #7
   ]}
  n27 = MaterialFunctionCall(X=@n1.R, Y=@n1.R, Z=@n6) {materialFunction=&MakeFloat3}
  n28 = GetMaterialAttributes(None=@n26) {attributeGetTypes [#8]}
  n29 = Add(
   A=@n28.World Position Offset
   B=MaterialFunctionCall() {materialFunction=&MF_FN_DestructibleDamage}.DamageAnimWPO
  )
  MaterialFunctionCall!(
   In=StaticSwitchParameter(
    True=SetMaterialAttributes!(MaterialAttributes=@n26, World Position Offset=@n27.Result) {attributeSetTypes [#8]}
    False=SetMaterialAttributes(MaterialAttributes=@n28.MaterialAttributes, World Position Offset=@n29) {attributeSetTypes [#8]}
   ) {parameterName=UseCameraPushDown}
   UseHeightFalloff=StaticBool()
   OnFireEmissiveOverride
   HeightFalloffRadius
   HeightFalloffHardness
   VariationMasksMobile
   In_Grass=StaticBool() {value=true}
   Is_Stylised_WE
  ) {materialFunction=/Game/Packages/Fortress_SharedMaterials/Base_Material_Functions/MF_Curie_Landscape}
  n30 = VectorParameter() {parameterName=Color4, channelNames {a=SSS Value, b="", g="", r=""}, defaultValue=<rgba 0.0856 0.1458 0.0167 0.2>}
  n31 = VectorParameter() {parameterName=Color3, channelNames {a=SSS Value, b="", g="", r=""}, defaultValue=<rgba 0.198 0.1624 0.0556 0.2>}
  n32 = VectorParameter() {parameterName=Color2, channelNames {a=SSS Value, b="", g="", r=""}, defaultValue=<rgba 0.2675 0.2969 0.0452 0.2>}
  n33 = VectorParameter() {parameterName=Color, channelNames {a=SSS Value, b="", g="", r=""}, defaultValue=<rgba 0.0907 0.1458 0.0167 0.2>}
  Multiply(
   A=MaterialFunctionCall(1=@n30, 2=@n31, 3=@n32, 4=@n33, A=NamedRerouteUsage()) {materialFunction=&Lerp_Multiple_Float3}.Lerp 4 Inputs
   B=MaterialFunctionCall(
    1=VectorParameter() {parameterName=BladeVariationColor_01, defaultValue=<rgba 0.3664 1 0.9001 1>}.RGB
    2=VectorParameter() {parameterName=BladeVariationColor_02, defaultValue=<rgba 0.7429 1 0.3061 1>}.RGB
    3=VectorParameter() {parameterName=BladeVariationColor_03, defaultValue=<rgba 1 0.7774 0.3915 1>}.RGB
    4=VectorParameter() {parameterName=BladeVariationColor_04, defaultValue=<rgba 0.4596 1 0.4362 1>}.RGB
    A=@n15
   ) {materialFunction=&Lerp_Multiple_Float3}.Lerp 4 Inputs
  )
  n34 = ComponentMask(None=TextureCoordinate()) {g=true}
  n35 = LinearInterpolate(
   A=ScalarParameter() {defaultValue=0.1, parameterName=MossBladeGradientOffset}
   B=ScalarParameter() {defaultValue=0.1, parameterName=MossBladeGradientOffset2}
   Alpha=NamedRerouteUsage()
  )
  n36 = Power(
   Base=OneMinus(None=Add(A=@n34, B=@n35))
   Exp=LinearInterpolate(
    A=ScalarParameter() {defaultValue=3, parameterName=GradientPow2}
    B=ScalarParameter() {defaultValue=2, parameterName=GradientPow}
    Alpha=NamedRerouteUsage()
   )
  )
  n37 = Power(
   Base=Add(
    A=TextureSampleParameter2D() {parameterName=FallBackAlpha, texture=&/T_Moss_AlphaBake}.G
    B=ScalarParameter() {parameterName=FallbackMossBladeGradientOffset}
   )
   Exp=ScalarParameter() {defaultValue=1, parameterName=FallbackGradientPow}
  )
  NamedRerouteDeclaration(None=NaniteReplace!(Default=Saturate(None=@n37), Nanite=Saturate(None=@n36))) {name=MossGradient, nodeColor=<rgba 0 0.7422 1 1>}
  NamedRerouteDeclaration(None=ScalarParameter() {defaultValue=1, parameterName=HexGridSize}) {name=HexGridSize, nodeColor=<rgba 1 0.5391 0 1>}
  NamedRerouteDeclaration(None=ScalarParameter() {defaultValue=64, parameterName=MacroScale}) {name=MacroScale, nodeColor=<rgba 0 1 0.1328 1>}
  NamedRerouteDeclaration() {name=useMipBias, nodeColor=<rgba 0 1 0.9531 1>}
  NamedRerouteDeclaration() {name=MipBiasValue, nodeColor=<rgba 0.6641 1 0 1>}
  NamedRerouteDeclaration(None=ComponentMask(None=@n8.Result) {r=true}) {name=ColorVariationNoise, nodeColor=<rgba 1 0 0.7031 1>}
  n38 = VertexColor()
  NamedRerouteDeclaration(None=@n38.G) {name=GradientVariation, nodeColor=<rgba 1 0.4688 0 1>}
  NamedRerouteDeclaration(None=@n38.R) {name=BladeColorVariation, nodeColor=<rgba 0 1 0.9531 1>}
  NamedRerouteDeclaration(None=ScalarParameter() {defaultValue=2, parameterName=HexGridSharpness}) {name=HexGridGradientSharpness, nodeColor=<rgba 0 0.7422 1 1>}
  NamedRerouteDeclaration(None=ScalarParameter() {defaultValue=4, parameterName=MinScale}) {name=MinScale, nodeColor=<rgba 1 0.8906 0 1>}
  NamedRerouteDeclaration(None=ScalarParameter() {defaultValue=8, parameterName=MaxScale}) {name=MaxScale, nodeColor=<rgba 1 0.8906 0 1>}
  NamedRerouteDeclaration(None=Constant() {r=-1}) {name=MinOffset, nodeColor=<rgba 0.7812 1 0 1>}
  NamedRerouteDeclaration(None=Constant() {r=1}) {name=MaxOffset, nodeColor=<rgba 0.7812 1 0 1>}
  NamedRerouteDeclaration(None=StaticBool()) {name=ForceLow, nodeColor=<rgba 0 1 0.3203 1>}
  NamedRerouteDeclaration(None=Constant()) {name=MinRotation, nodeColor=<rgba 0.5938 1 0 1>}
  NamedRerouteDeclaration(None=Constant()) {name=MaxRotation, nodeColor=<rgba 0.5938 1 0 1>}
  NaniteReplace(Default=@n20.RGB, Nanite=@n21.RGB)
  ObjectPositionWS()
  ScalarParameter() {defaultValue=-200, parameterName=debug-grass-wpo-pushdown-test33amt}
  SetMaterialAttributes!(MaterialAttributes, Base Color=TextureSample() {texture=/Game/Packages/Fortress_SharedMaterials/Textures/T_UV_Checker_D}.RGB) {attributeSetTypes [#1]}
  StaticBool()
  VertexNormalWS()
  WorldPosition()
 ]
}
