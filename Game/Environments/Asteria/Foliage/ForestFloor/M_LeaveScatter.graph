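# Masked, two-sided leaf-scatter material for instanced forest-floor meshes that animates leaves being picked up by wind.
# UV1 supplies per-leaf pivots, which seed random time offsets for a looping up-and-back timeline.
# MF_WindActorData wind speed and direction drive rotation, sideways drift and lift, while a fade term pulls vertices toward a point.
# Vertex red compared against stepped time levels selects which leaves move each cycle, with squared wind intensity adding wiggle.
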Material /Game/Environments/Asteria/Foliage/ForestFloor/M_LeaveScatter
graph {
 properties {bUsedWithInstancedStaticMeshes=true, blendMode=BLEND_Masked, pixelDepthOffsetMode=PDOM_Legacy, twoSided=true, usedShadingModels=Default Lit}
 nodes [
  n1 = WorldPosition()
  n2 = MaterialFunctionCall!(
   Position=Divide(A=@n1.XYZ, B=ScalarParameter() {defaultValue=0.1, parameterName=WindSize})
   IsTrees
   IsGrass
   IsMediumPlant
   IsMediumPlant_2024Anim
   IsMediumPlant2ndLevel
   MediumPlant_BasePivotPostion
  ) {materialFunction=/Game/Environments/Asteria/Foliage/Wind/MF_WindActorData}
  n3 = ScalarParameter() {defaultValue=0.1207, parameterName=maxRotation1}
  n4 = ScalarParameter() {defaultValue=1, parameterName=maxRotation2}
  n5 = RotateAboutAxis(
   NormalizedRotationAxis=Normalize(VectorInput=Constant3Vector() {constant=<rgba 1 1 1 0>})
   RotationAngle=LinearInterpolate(B=LinearInterpolate(A=@n3, B=@n4, Alpha=@n2.Speed 0-1), Alpha=NamedRerouteUsage())
   PivotPoint=NamedRerouteUsage()
   Position=@n1.XYZ
  )
  n6 = Multiply(A=Add(A=Time(), B=NamedRerouteUsage()), B=ScalarParameter() {defaultValue=0.3393, parameterName=PhaseLength})
  n7 = Frac(None=@n6)
  n8 = Subtract(A=@n7) {constB=0.5}
  n9 = Sign(None=@n8)
  n10 = Transform(None=Constant3Vector() {constant=<rgba 0 0 1 0>}) {transformSourceType=TRANSFORMSOURCE_Instance}
  n11 = Multiply(
   A=MaterialFunctionCall!(VectorToRotate=Multiply(A=@n2.WindI_Direction, B=@n9), LookAtVector=@n10, RestingVector) {materialFunction=/Engine/Functions/Engine_MaterialFunctions01/Vectors/RotateVector}.Result
   B=LinearInterpolate(
    A=ScalarParameter() {defaultValue=99.6528, parameterName=DirectionOffset1}
    B=ScalarParameter() {defaultValue=419.2172, parameterName=DirectionOffset2}
    Alpha=@n2.Speed 0-1
   )
  )
  n12 = LinearInterpolate(
   A=ScalarParameter() {defaultValue=99.1954, parameterName=HeightOffset}
   B=ScalarParameter() {defaultValue=99.1954, parameterName=HeightOffset}
   Alpha=VertexColor().R
  )
  n13 = Ceil(None=@n8)
  n14 = LinearInterpolate(B=ScalarParameter() {defaultValue=0.9566, parameterName=HeightReturnMult}, Alpha=@n13) {constA=1}
  n15 = ScalarParameter() {defaultValue=1.2, parameterName=HeightStrengthMult}
  n16 = Multiply(A=@n10, B=Multiply(A=Multiply(A=@n12, B=@n14), B=LinearInterpolate(B=@n15, Alpha=@n2.Speed 0-1)))
  n17 = NamedRerouteUsage()
  n18 = WorldPosition()
  n19 = Add(
   A=Multiply(A=Add(A=@n11, B=@n16), B=NamedRerouteUsage())
   B=LinearInterpolate(A=Constant3Vector(), B=Subtract(A=@n17, B=@n18.XYZ), Alpha=NamedRerouteUsage())
  )
  n20 = VertexColor()
  n21 = Frac(None=Divide(A=Floor(None=@n6), B=ScalarParameter() {defaultValue=10, parameterName=NumberOfLevels}))
  n22 = Subtract(
   A=@n21
   B=LinearInterpolate(
    A=ScalarParameter() {defaultValue=-0.9564, parameterName=AmountOfLeavesToPickup}
    B=ScalarParameter() {defaultValue=-0.627, parameterName=AmountOfLeavesToPickup2}
    Alpha=@n2.Speed 0-1
   )
  )
  n23 = OneMinus(None=@n21)
  n24 = Multiply(A=Ceil(None=Subtract(A=@n20.R, B=@n22)), B=Floor(None=Add(A=@n20.R, B=@n23)))
  Add(
   A=Multiply(A=ScalarParameter() {defaultValue=8.0156, parameterName=WindWiggle}, B=Multiply(A=@n2.WindIntesnity, B=@n2.WindIntesnity))
   B=Multiply(A=Add(A=@n5, B=@n19), B=Saturate(None=@n24))
  )
  LinearInterpolate(A=OneMinus(None=@n7), B=@n7, Alpha=@n13)
  Multiply(A=TwoSidedSign(), B=TextureSample() {samplerType=SAMPLERTYPE_Normal, texture=&/Leaves_N}.RGB)
  n25 = TextureCoordinate() {coordinateIndex=1}
  n26 = TransformPosition(None=MaterialFunctionCall(X=ComponentMask(None=@n25) {r=true}, Y=ComponentMask(None=@n25) {g=true}, Z=Constant()) {materialFunction=&MakeFloat3}.Result) {transformSourceType=TRANSFORMPOSSOURCE_Instance, transformType=TRANSFORMPOSSOURCE_World}
  NamedRerouteDeclaration(None=@n26) {name=Pivot, nodeColor=<rgba 1 0.8672 0 1>}
  n27 = NamedRerouteUsage()
  n28 = Add(A=ComponentMask(None=@n27) {r=true}, B=ComponentMask(None=@n27) {g=true})
  n29 = ScalarParameter() {defaultValue=33.3, parameterName=Random}
  n30 = Multiply(A=Frac(None=Divide(A=@n28, B=@n29)), B=ScalarParameter() {defaultValue=193943.0781, parameterName=RandomTimeOffset})
  NamedRerouteDeclaration(None=@n30) {name=RandomPerleaf, nodeColor=<rgba 0.4062 0 1 1>}
  n31 = Frac(None=Multiply(A=@n7) {constB=2})
  n32 = LinearInterpolate(A=@n31, B=OneMinus(None=@n31), Alpha=@n13)
  n33 = Power(Base=@n32, Exp=ScalarParameter() {defaultValue=1.5435, parameterName=Accel})
  NamedRerouteDeclaration(None=@n33) {name=TimeLineAnim, nodeColor=<rgba 0.1953 0 1 1>}
  n34 = ScalarParameter() {defaultValue=16, parameterName=Fadeout}
  NamedRerouteDeclaration(None=Saturate(None=Power(Base=@n32, Exp=@n34))) {name=TimeLineScale, nodeColor=<rgba 0.7344 0 1 1>}
  Reroute(None=@n26)
  Reroute(None=@n30)
  Reroute(None=@n33)
  TextureSample() {texture=&/Leaves_D}
  TextureSample() {samplerType=SAMPLERTYPE_Masks, texture=&/Leaves_SOR}
 ]
}
