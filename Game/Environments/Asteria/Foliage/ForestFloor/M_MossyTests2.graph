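# Masked two-sided foliage test material for mossy forest-floor plants in Asteria, enabled for Nanite and instanced static meshes.
# Distance-field AO blends two tint colors, bypassed on mobile, with Fresnel on vertex normals brightening edges.
# An offset UV V gradient shapes brightness, roughness, specular and subsurface tint.
# GrassySpec mask and normal textures add detail, with normals flipped by two-sided sign.
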Material /Game/Environments/Asteria/Foliage/ForestFloor/M_MossyTests2
graph {
 properties {
  bUsedWithInstancedStaticMeshes=true
  bUsedWithNanite=true
  blendMode=BLEND_Masked
  pixelDepthOffsetMode=PDOM_Legacy
  shadingModel=MSM_TwoSidedFoliage
  twoSided=true
  usedShadingModels=Two Sided Foliage
 }
 nodes [
  n1 = VertexNormalWS()
  n2 = VectorParameter() {parameterName=Color2, channelNames {a=SSS Value, b="", g="", r=""}, defaultValue=<rgba 0.2811 0.2969 0.0452 0.2>}
  n3 = VectorParameter() {parameterName=Color, channelNames {a=SSS Value, b="", g="", r=""}, defaultValue=<rgba 0.0907 0.1458 0.0167 0.2>}
  n4 = DistanceFieldApproxAO!(
   World Position
   Normal
   BaseDistance=ScalarParameter() {defaultValue=155, group=C_Dirt_AO, parameterName=Details AO Distance, sortPriority=22}
   Radius
  ) {numSteps=3}
  n5 = Constant() {r=1}
  n6 = LinearInterpolate(A=@n2, B=@n3, Alpha=Saturate(None=ShadingPathSwitch(Default=@n4, Mobile=@n5)))
  n7 = ComponentMask(None=Subtract(A=TextureCoordinate(), B=ScalarParameter() {defaultValue=0.25})) {g=true}
  n8 = OneMinus(None=@n7)
  Add(A=Multiply(A=Fresnel(Normal=@n1) {exponent=2.5}, B=@n6), B=Multiply(A=@n6, B=LinearInterpolate(Alpha=@n8) {constA=0.25}))
  LinearInterpolate(
   A=Multiply(
    A=ScalarParameter() {defaultValue=0.2, parameterName=Roughness}
    B=TextureSample() {samplerType=SAMPLERTYPE_Masks, texture=&/T_GrassySpec_M}.R
   )
   Alpha=@n7
  )
  Multiply(
   A=Multiply(A=@n6, B=Multiply(A=@n3.SSS Value, B=@n8))
   B=TextureSample() {samplerType=SAMPLERTYPE_Masks, texture=&T_GrassySpec_M}.R
  )
  Multiply(A=ScalarParameter() {defaultValue=0.035, parameterName=Specular}, B=@n8)
  Multiply(
   A=TextureSample() {samplerType=SAMPLERTYPE_Normal, texture=&/T_GrassySpec_N}.RGB
   B=TwoSidedSign()
  )
  ScalarParameter() {defaultValue=0.5, parameterName=SSS Thickness}
 ]
}
