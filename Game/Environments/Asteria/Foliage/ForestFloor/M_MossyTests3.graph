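# Masked, two-sided foliage material for forest-floor moss, apparently a test, enabled for Nanite and instanced meshes.
# Distance field AO blends a dirt color into the moss color, which is tinted on back faces; Fresnel adds a rim.
# A UV-V gradient modulates color, roughness, specular, subsurface color and moss normal strength.
# The graph also holds a panning world-aligned noise normal with no visible consumer.
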Material /Game/Environments/Asteria/Foliage/ForestFloor/M_MossyTests3
graph {
 properties {
  bUsedWithInstancedStaticMeshes=true
  bUsedWithNanite=true
  blendMode=BLEND_Masked
  pixelDepthOffsetMode=PDOM_Legacy
  shadingModel=MSM_TwoSidedFoliage
  twoSided=true
  usedShadingModels=Two Sided Foliage
 }
 nodes [
  n1 = VertexNormalWS()
  n2 = VectorParameter() {parameterName=Color2, channelNames {a=SSS Value, b="", g="", r=""}, defaultValue=<rgba 0.2396 0.1362 0.0462 0.2>}
  n3 = VectorParameter() {parameterName=Color, channelNames {a=SSS Value, b="", g="", r=""}, defaultValue=<rgba 0.1393 0.224 0.0257 0.2>}
  n4 = LinearInterpolate(
   A=Multiply(A=VectorParameter() {parameterName=Param_3, defaultValue=<rgba 0.6288 1.0274 1.5 1>}.RGB, B=@n3) {constA=0.5}
   B=@n3
   Alpha=Saturate(None=TwoSidedSign())
  )
  n5 = Saturate(None=DistanceFieldApproxAO!(
    World Position
    Normal
    BaseDistance=ScalarParameter() {defaultValue=155, group=C_Dirt_AO, parameterName=Details AO Distance, sortPriority=22}
    Radius
   ) {numSteps=3})
  n6 = LinearInterpolate(A=@n2, B=@n4, Alpha=@n5)
  n7 = ComponentMask(None=Subtract(A=TextureCoordinate(), B=ScalarParameter() {defaultValue=-0.2})) {g=true}
  n8 = OneMinus(None=@n7)
  Add(A=Multiply(A=Fresnel(Normal=@n1) {exponent=2.5}, B=@n6), B=Multiply(A=@n6, B=LinearInterpolate(Alpha=@n8) {constA=0.25}))
  Constant3Vector()
  LinearInterpolate(
   A=ScalarParameter() {defaultValue=0.1, parameterName=AOR}
   B=LinearInterpolate(A=ScalarParameter() {defaultValue=0.15, parameterName=Roughness}, Alpha=@n7)
   Alpha=@n5
  )
  MaterialFunctionCall() {materialFunction=&WorldAlignedNormal}
  n9 = TextureObject() {samplerType=SAMPLERTYPE_Normal, texture=/Game/Characters/Player/Female/Medium/Bodies/F_MED_Bistro_Astronaut/FX/T_GaussianNoise_Normal}
  n10 = VectorParameter() {parameterName=Param_2}
  MaterialFunctionCall!(
   TextureObject=@n9
   TextureSize=ScalarParameter() {parameterName=sca}
   Normal
   WorldSpace
   ProjectionTransitionContrast
   WorldPosition
   Use High Quality Normals
   Pan_YZ=Multiply(A=Time(), B=ComponentMask(None=@n10.RGB) {g=true, r=true})
   Pan_XZ
   Pan_XY
  ) {materialFunction=/Game/Weapons/WeaponSkins/Wraps/MaterialFunctions/WorldAlignedNormal_ComplexPan}
  Multiply(A=@n4, B=Multiply(A=@n3.SSS Value, B=@n8))
  Multiply(A=ScalarParameter() {defaultValue=0.2, parameterName=Specular}, B=@n8)
  n11 = TextureSample() {samplerType=SAMPLERTYPE_Normal, texture=&/Moss_Simple_N}
  n12 = VectorParameter() {defaultValue=<rgba 0.5 0.5 1 0>}
  Multiply(A=LinearInterpolate(A=Constant3Vector() {constant=<rgba 0 0 1 0>}, B=Multiply(A=@n11.RGB, B=@n12.RGB), Alpha=@n8), B=TwoSidedSign())
  ScalarParameter() {defaultValue=0.2, parameterName=SSS Thickness}
  VectorParameter() {parameterName=Param_1}
 ]
}
