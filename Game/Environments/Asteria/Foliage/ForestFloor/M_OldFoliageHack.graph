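# M_OldFoliageHack is a masked two-sided foliage material for Asteria forest-floor plants, using Nightmare jungle plant textures.
# TwoSidedSign varies specular, roughness and subsurface per face.
# A detail-normal layer blends with a darkened wet layer via noise HeightLerp, then with a flat mud-colored layer via distance-field AO.
# An optional panning-noise wind adds normal and world-position offset, masked by UV or vertex color, and is skipped on mobile.
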Material /Game/Environments/Asteria/Foliage/ForestFloor/M_OldFoliageHack
graph {
 properties {
  bUseMaterialAttributes=true
  blendMode=BLEND_Masked
  pixelDepthOffsetMode=PDOM_Legacy
  shadingModel=MSM_FromMaterialExpression
  twoSided=true
  usedShadingModels=Default Lit | Two Sided Foliage
 }
 nodes [
  n1 = TextureSampleParameter2D() {group=Tex, parameterName=D, texture=/Game/Environments/Apollo/Foliage/Nightmare/Textures/T_Nightmare_JunglePlants_01_D}
  n2 = TextureSampleParameter2D() {parameterName=S, samplerType=SAMPLERTYPE_Masks, texture=/Game/Environments/Apollo/Foliage/Nightmare/Textures/T_Nightmare_JunglePlants_01_S}
  n3 = ScalarParameter() {defaultValue=0.5, group=S, parameterName=S Inside}
  n4 = ScalarParameter() {defaultValue=0.25, group=S, parameterName=S Out}
  n5 = TwoSidedSign()
  n6 = ScalarParameter() {defaultValue=1, group=R, parameterName=R Inside}
  n7 = ScalarParameter() {defaultValue=0.65, group=R, parameterName=R Out}
  n8 = VectorParameter() {parameterName=SSS Inside, defaultValue=<rgba 1 0.8555 0.2249 1>}
  n9 = VectorParameter() {group=SSS, parameterName=SSS Tone, defaultValue=<rgba 2 1.438 0.4285 0.5>}
  n10 = TextureSampleParameter2D() {parameterName=N, samplerType=SAMPLERTYPE_Normal, texture=/Game/Environments/Apollo/Foliage/Nightmare/Textures/T_Nightmare_JunglePlants_01_N}
  n11 = MakeMaterialAttributes!(
   BaseColor=@n1.RGB
   Metallic
   Specular=Multiply(A=@n2.R, B=LinearInterpolate(A=@n3, B=@n4, Alpha=@n5))
   Roughness=Multiply(A=@n2.B, B=LinearInterpolate(A=@n6, B=@n7, Alpha=@n5))
   Anisotropy
   EmissiveColor
   Opacity=LinearInterpolate(A=@n8.A, B=@n9.A, Alpha=@n5)
   OpacityMask=@n1.RGB
   Normal=Multiply(A=@n5, B=@n10.RGB)
   Tangent
   WorldPositionOffset
   SubsurfaceColor=LinearInterpolate(A=Multiply(A=@n1.RGB, B=@n8.RGB), B=Multiply(A=@n1.RGB, B=@n9.RGB), Alpha=@n5)
   ClearCoat
   ClearCoatRoughness
   AmbientOcclusion
   Refraction
   CustomizedUVs_0
   CustomizedUVs_1
   CustomizedUVs_2
   CustomizedUVs_3
   CustomizedUVs_4
   CustomizedUVs_5
   CustomizedUVs_6
   CustomizedUVs_7
   PixelDepthOffset
   ShadingModel=ShadingModel() {shadingModel=MSM_TwoSidedFoliage}
   Displacement
  )
  n12 = ComponentMask(None=VertexNormalWS()) {b=true}
  n13 = SetMaterialAttributes!(
   MaterialAttributes=GetMaterialAttributes(None=@n11) {attributeGetTypes [#1, #2, #3]}.MaterialAttributes
   Base Color=VectorParameter() {group=Jungle, parameterName=WetRock, defaultValue=<rgba 0.2448 0.1103 0.0407 1>}.RGB
   Specular=ScalarParameter() {defaultValue=0.25, parameterName=JS}
   Roughness=LinearInterpolate(
    A=ScalarParameter() {defaultValue=0.7, group=Jungle, parameterName=JR}
    B=ScalarParameter() {defaultValue=0.7, group=Jungle, parameterName=JRTop}
    Alpha=Saturate(None=@n12)
   )
   Subsurface Color=Constant()
   Shading Model=ShadingModel()
  ) {attributeSetTypes [
    #1
    #2
    #3
    #4
    #5
   ]}
  n14 = GetMaterialAttributes(None=@n11) {attributeGetTypes [#6, #3]}
  n15 = TextureSample(UVs=Multiply(A=TextureCoordinate(), B=ScalarParameter() {defaultValue=5, parameterName=Tile Details})) {samplerType=SAMPLERTYPE_Normal, texture=/Game/Environments/Asteria/Rocks/Textures/T_RockDetails2_N}
  n16 = VectorParameter() {parameterName=Details Value, defaultValue=<rgba 0.25 0.25 0 1>}
  n17 = SetMaterialAttributes!(MaterialAttributes=@n14.MaterialAttributes, Normal=Add(A=@n14.Normal, B=Multiply(A=@n15.RGB, B=@n16.RGB)), Roughness=@n14.MaterialAttributes) {attributeSetTypes [#6, #3]}
  n18 = GetMaterialAttributes(None=@n17) {attributeGetTypes [#1, #6]}
  n19 = Multiply(
   A=TextureSample() {samplerType=SAMPLERTYPE_Normal, texture=/Game/Global/Textures/Default/Noise/T_SwirlNoise_N}.RGB
   B=VectorParameter() {parameterName=WetN, defaultValue=<rgba 1 1 0 0>}.RGB
  )
  n20 = TwoSidedSign()
  n21 = SetMaterialAttributes!(
   MaterialAttributes=@n18.MaterialAttributes
   Normal=Add(A=@n18.Normal, B=Multiply(A=@n19, B=@n20))
   Base Color=Multiply(A=@n18.Base Color) {constB=0.5}
   Shading Model=ShadingModel() {shadingModel=MSM_TwoSidedFoliage}
   Roughness=Constant()
  ) {attributeSetTypes [
    #6
    #1
    #5
    #3
   ]}
  n22 = TextureCoordinate()
  n23 = ScalarParameter() {defaultValue=5, parameterName=Param_3}
  n24 = MaterialFunctionCall(
   Transition Phase=ScalarParameter() {defaultValue=0.372}
   Height Texture=TextureSample(UVs=Multiply(A=@n22, B=@n23)) {samplerType=SAMPLERTYPE_LinearGrayscale, texture=/Game/Global/Textures/Default/Noise/T_SwirlNoise_N_height}.R
   Contrast=ScalarParameter() {defaultValue=5, parameterName=Param_1}
  ) {materialFunction=&HeightLerp}
  n25 = TwoSidedSign()
  n26 = DistanceFieldApproxAO!(
   World Position
   Normal
   BaseDistance=ScalarParameter() {defaultValue=15, group=Jungle, parameterName=Mud AO Distance, sortPriority=22}
   Radius=ScalarParameter() {defaultValue=155, group=Jungle, parameterName=Mud AO Fade, sortPriority=23}
  ) {numSteps=3}
  n27 = ScalarParameter() {defaultValue=0.35, parameterName=Mud Transparency}
  n28 = BlendMaterialAttributes(
   A=@n13
   B=BlendMaterialAttributes(A=@n17, B=@n21, Alpha=Multiply(A=@n24.Alpha, B=@n25)) {vertexAttributeBlendType=UseA}
   Alpha=Saturate(None=Add(A=@n26, B=@n27))
  )
  MaterialFunctionCall!(In=@n28) {materialFunction=/Game/Athena/WinterArt/Materials/MF_WindWPO}
  Multiply()
  n29 = GetMaterialAttributes(None=@n28) {attributeGetTypes [#6, #7]}
  n30 = StaticSwitchParameter(
   True=StaticSwitchParameter(True=TextureCoordinate() {coordinateIndex=4}, False=TextureCoordinate() {coordinateIndex=3}) {parameterName=WindUse UV4}
   False=TextureCoordinate() {coordinateIndex=2}
  ) {parameterName=WindUse UV3}
  n31 = ComponentMask(None=AppendVector(
    A=ScalarParameter() {defaultValue=0.4, group=Wind, parameterName=UV_Wind_X, sortPriority=2}
    B=ScalarParameter() {defaultValue=0.2, group=Wind, parameterName=UV_Wind_Y, sortPriority=2}
   )) {g=true, r=true}
  n32 = ScalarParameter() {defaultValue=-0.5, group=Wind, parameterName=WindSpeed_X, sortPriority=1}
  n33 = ScalarParameter() {group=Wind, parameterName=WindSpeed_Y, sortPriority=1}
  n34 = TextureSampleParameter2D(UVs=Panner(Coordinate=Multiply(A=@n30, B=@n31), Speed=AppendVector(A=@n32, B=@n33))) {group=Wind, parameterName=WindTexture, samplerType=SAMPLERTYPE_Normal, texture=&T_SwirlNoise_N}
  n35 = Add(A=@n29.Normal, B=Multiply(A=@n34.RGB, B=Constant3Vector() {constant=<rgba 1 1 0 0>}))
  n36 = ComponentMask(None=TextureCoordinate() {coordinateIndex=2}) {r=true}
  n37 = VertexColor()
  n38 = Multiply(
   A=LinearInterpolate(A=ScalarParameter() {group=Wind, parameterName=UV Ramp Min Weight}, Alpha=@n36)
   B=StaticComponentMaskParameter(None=AppendVector(A=@n37, B=@n37)) {defaultA=true, group=Wind, parameterName=ChannelSelect}
  )
  n39 = StaticSwitchParameter(True=@n38, False=@n36) {group=Wind, parameterName=UseVertexColorsMask}
  n40 = AppendVector(
   A=ScalarParameter() {defaultValue=5, group=Wind, parameterName=WPO_Wind_X, sortPriority=3}
   B=ScalarParameter() {group=Wind, parameterName=WPO_Wind_Y, sortPriority=3}
  )
  n41 = ScalarParameter() {defaultValue=-4, group=Wind, parameterName=WPO_Wind_Z, sortPriority=3}
  n42 = Multiply(
   A=StaticSwitchParameter(True=@n38, False=@n36) {group=Wind, parameterName=UseVertexColorsMask}
   B=Multiply(A=@n34.G, B=AppendVector(A=@n40, B=@n41))
  )
  n43 = Reroute()
  n44 = StaticSwitchParameter(
   True=SetMaterialAttributes!(
    MaterialAttributes=@n29.MaterialAttributes
    Normal=LinearInterpolate(A=@n29.Normal, B=@n35, Alpha=@n39)
    World Position Offset=Add(A=@n29.World Position Offset, B=@n42)
   ) {attributeSetTypes [#6, #7]}
   False=@n43
  ) {group=Wind, parameterName=WindWPO}
  ShadingPathSwitch(Default=@n44, Deferred=@n44, Forward=@n44, Mobile=@n43)
 ]
}
